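# Sound cue for the looping night ambience of the Farm point of interest in Asteria.
# A quality-level node picks between two looping wave players: the quad wave for the first two levels and the stereo wave for the last two.
# It plays at reduced volume through the outdoor ambient sound class with a shared Athena ambient attenuation asset.
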SoundCue /Game/Sounds/Asteria/POI/Farm/Ambience_Farm_Loop_Night_Cue
root=QualityLevel_0
nodes [
 {id=QualityLevel_0, children [WavePlayer_4,WavePlayer_4,WavePlayer_5,WavePlayer_5]}
 {id=WavePlayer_4, properties {bLooping=true, soundWaveAssetPtr=&/Ambience_Farm_Loop_Night_Quad}}
 {id=WavePlayer_5, properties {bLooping=true, soundWaveAssetPtr=&/Ambience_Farm_Loop_Night_Stereo}}
]
properties {
 attenuationSettings=&Fort_Ambient_Athena_Att
 maxDistance=2097152
 soundClassObject=&Fort_Ambient_Outdoors_Athena
 volumeMultiplier=0.301
 preEffectBusSends [{}]
}
